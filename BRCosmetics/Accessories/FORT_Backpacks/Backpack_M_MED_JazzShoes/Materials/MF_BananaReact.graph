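# Material function for the JazzShoes backpack's banana face, switching eye and brow expressions by collapsing masked vertices toward the object position through World Position Offset.
# Pose parameters pass thresholds to choose vertex color values, fed to VertexColorMask for the eye and brow masks.
# An optional GPU blink alternates two time-driven sine pulses to replace the blink pose.
# PeelyBlenderFace can swap in the dizzy-face mask before the offset joins the incoming material attributes.
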MaterialFunction /BRCosmetics/Accessories/FORT_Backpacks/Backpack_M_MED_JazzShoes/Materials/MF_BananaReact
graph {nodes [
  Constant()
  Constant() {r=1}
  n1 = GetMaterialAttributes(None=FunctionInput() {inputName=Material, inputType=FunctionInput_MaterialAttributes}) {attributeGetTypes [#1, #2]}
  n2 = Subtract(A=WorldPosition().XYZ, B=ObjectPositionWS())
  n3 = MaterialFunctionCall!(
   Hide Element Toggle=Constant() {r=1}
   VertexColorChannel=NamedRerouteUsage()
   Vertex Value to hide=ScalarParameter() {group=Banana, parameterName="DizzyFaceValue(R)"}
  ) {materialFunction=/Game/Packages/Fortress_SharedMaterials/Base_Material_Functions/VertexColorMask}
  n4 = Constant() {r=0.95}
  n5 = Constant() {r=1}
  n6 = Multiply(A=Time(), B=ScalarParameter() {defaultValue=0.25, group=Banana, parameterName=BlinkSpeed01})
  n7 = Add(A=ConstantBiasScale(None=Sine(None=@n6)), B=ScalarParameter() {defaultValue=-0.99, group=Banana, parameterName=UpperSize01})
  n8 = ScalarParameter() {defaultValue=0.3, group=Banana, parameterName=BlinkSpeed02}
  n9 = Multiply(A=Time(), B=@n8)
  n10 = Add(A=ConstantBiasScale(None=Sine(None=@n9)), B=ScalarParameter() {defaultValue=-0.995, group=Banana, parameterName=UpperSize02})
  n11 = Floor(None=Multiply(A=Time(), B=@n8))
  n12 = Constant() {r=2}
  n13 = LinearInterpolate(A=Saturate(None=Sign(None=@n7)), B=Saturate(None=Sign(None=@n10)), Alpha=Saturate(None=Fmod(A=@n11, B=@n12)))
  n14 = NamedRerouteUsage()
  n15 = MaterialFunctionCall(
   Hide Element Toggle=Constant() {r=1}
   VertexColorChannel=NamedRerouteUsage()
   Vertex Value to hide=ScalarParameter() {defaultValue=0.3, group=Banana, parameterName="BananaHappyPoseValue(R)"}
  ) {materialFunction=&VertexColorMask}
  n16 = If(
   A=ScalarParameter() {group=Banana_Poses, parameterName=R_squint_inner_pose}
   B=ScalarParameter() {defaultValue=0.375, group=Banana, parameterName=Threshold_squint_inner_pose}
   "A > B"=MaterialFunctionCall(
    Hide Element Toggle=Constant() {r=1}
    VertexColorChannel=@n14
    Vertex Value to hide=ScalarParameter() {defaultValue=0.5, group=Banana, parameterName="BananaSquintPoseValue(R)"}
   ) {materialFunction=&VertexColorMask}.Result
   "A == B"=@n15.Result
   "A < B"=@n15.Result
  )
  n17 = If(
   A=LinearInterpolate(
    A=ScalarParameter() {group=Banana_Poses, parameterName=R_blink_pose}
    B=Step(X=@n13) {constY=0.2}
    Alpha=ScalarParameter() {defaultValue=1, group=Banana, parameterName=UseGPUBlink}
   )
   B=ScalarParameter() {defaultValue=0.1, group=Banana, parameterName=Threshold_R_blink_pose}
   "A > B"=MaterialFunctionCall(
    Hide Element Toggle=Constant() {r=1}
    VertexColorChannel=@n14
    Vertex Value to hide=ScalarParameter() {defaultValue=0.6, group=Banana, parameterName="BananaBlinkPoseValue(R)"}
   ) {materialFunction=&VertexColorMask}.Result
   "A == B"=@n16
   "A < B"=@n16
  )
  n18 = NamedRerouteUsage()
  n19 = Constant() {r=0.95}
  n20 = Constant() {r=1}
  n21 = ScalarParameter() {defaultValue=1, group=Banana, parameterName=sdfaafafafafa}
  n22 = If(
   A=ScalarParameter() {group=Banana_Poses, parameterName=R_brow_up_pose}
   B=ScalarParameter() {defaultValue=0.1, group=Banana, parameterName=Threshold_R_brow_up_pose}
   "A > B"=ScalarParameter() {defaultValue=0.4, group=Banana, parameterName="BananaBrowUpValue(R)"}
   "A == B"=@n21
   "A < B"=@n21
  )
  n23 = If(
   A=ScalarParameter() {group=Banana_Poses, parameterName=C_glabella_down_pose}
   B=ScalarParameter() {defaultValue=0.1, group=Banana, parameterName=Threshold_Glabella_down}
   "A > B"=ScalarParameter() {defaultValue=0.2, group=Banana, parameterName="GlabellaDownValue(G)"}
   "A == B"=@n21
   "A < B"=@n21
  )
  n24 = ScalarParameter() {group=Banana_Poses, parameterName=C_glabella_up_pose}
  n25 = ScalarParameter() {defaultValue=0.1, group=Banana, parameterName=Threshold_Glabella_up}
  n26 = Constant() {r=0.1}
  n27 = MaterialFunctionCall(
   Hide Element Toggle=Constant() {r=1}
   VertexColorChannel=NamedRerouteUsage()
   Vertex Value to hide=Min(A=Min(A=@n22, B=@n23), B=If(A=@n24, B=@n25, "A > B"=@n26, "A == B"=@n21, "A < B"=@n21))
  ) {materialFunction=&VertexColorMask}
  n28 = NamedRerouteUsage()
  n29 = OneMinus(None=Max(A=If(A=@n28, B=@n19, "A > B"=@n20, "A == B"=@n27.Result, "A < B"=@n27.Result), B=If(A=@n18, B=@n4, "A > B"=@n5, "A == B"=@n17, "A < B"=@n17)))
  n30 = If(
   A=Saturate(None=ScalarParameter() {parameterName=PeelyBlenderFace})
   B=NamedRerouteUsage()
   "A > B"=Saturate(None=OneMinus(None=@n3.Result))
   "A == B"=@n29
   "A < B"=@n29
  )
  n31 = Add(A=@n1.WorldPositionOffset, B=Multiply(A=Multiply(A=@n2) {constB=-1}, B=Saturate(None=@n30)))
  FunctionOutput(None=SetMaterialAttributes!(
    MaterialAttributes=@n1.MaterialAttributes
    WorldPositionOffset=PreviousFrameSwitch(Current Frame=@n31, Previous Frame=@n31)
    BaseColor=@n1.MaterialAttributes
   ) {attributeSetTypes [#1, #2]})
  n32 = VertexColor()
  NamedRerouteDeclaration(None=ChannelMaskParameter(None=@n32) {group=Banana, parameterName=Banana_Former Red}.R) {name=EyeVtxColor, nodeColor=<rgba 0.4297 0 1 1>}
  NamedRerouteDeclaration(None=ChannelMaskParameter(None=@n32) {group=Banana, maskChannel=Green, parameterName=Banana_Former Green, defaultValue=<rgba 0 1 0 0>}.R) {name=EyeBrowVtxColor, nodeColor=<rgba 0 0.5312 1 1>}
  ScalarParameter() {defaultValue=0.1, group=Banana, parameterName=BananaEyeThreshold}
 ]}
