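# Master material for the Jazz Shoes backpack cosmetic, using material attributes and an expression-driven shading model on skeletal meshes with mesh deformer support.
# The shared 2023 character PreFX function builds the base material, which passes through two static switches, both on by default.
# The first applies the BananaReact function; the second applies SubUV textures, masked by a selectable Tech Art Mask channel.
# The result feeds the shared character PostFX function.
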Material /BRCosmetics/Accessories/FORT_Backpacks/Backpack_M_MED_JazzShoes/Materials/M_JazzShoes_Backpack
graph {
 properties {
  bUseMaterialAttributes=true
  bUsedWithMeshDeformer=true
  bUsedWithSkeletalMesh=true
  pixelDepthOffsetMode=PDOM_Legacy
  shadingModel=MSM_FromMaterialExpression
  usedShadingModels=Default Lit | Subsurface | Clear Coat | Subsurface Profile | Cloth
 }
 nodes [
  n1 = MaterialFunctionCall() {materialFunction=&MF_CharacterParent_2023_PreFX}
  n2 = ChannelMaskParameter(None=@n1.Tech Art Mask) {group=Skin FX, parameterName=Sub UV Mask Channel}
  n3 = StaticSwitchParameter(
   True=MaterialFunctionCall!(Material=@n1.Result) {materialFunction=&/MF_BananaReact}.Result
   False=@n1.Result
  ) {defaultValue=true, group=Skin FX, parameterName=Use Banana}
  MaterialFunctionCall(
   Material=StaticSwitchParameter(
    True=MaterialFunctionCall!(Alpha=@n2.R, Material=@n3) {materialFunction=/Game/Weapons/WeaponSkins/MaterialFunction/SubUVTextures}.Result
    False=@n3
   ) {defaultValue=true, group=Skin FX, parameterName=Use Sub UV texture}
   M=@n1.M
   Pre Skinned Local Position=@n1.M
   Distance Blend=@n1.M
   Fresnel=@n1.M
  ) {materialFunction=&MF_CharacterParent_2023_PostFX}
 ]
}
